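# Sound cue for a Season 20 Battle Pass Week 6 voice-over line spoken by The Imagined.
# A single Dialogue Player node plays the matching dialogue wave, with The Imagined as speaker and no targets.
# The cue uses the Season 20 quest attenuation, routes through the narrator VO sound class and has a raised volume multiplier.
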SoundCue /BattlePassS20/Audio/VO/Cues/Week_6/02/IMAGINED_2022_REGS_Week06_02_01_010_Dialogue_Cue
root=DialoguePlayer_0
nodes [{
  id=DialoguePlayer_0
  properties {dialogueWaveParameter {
    dialogueWave=/BattlePassS20/Audio/VO/Waves/Imagined/IMAGINED_2022_REGS_Week06_02_01_010_Dialogue
    context {speaker=/BattlePassS20/Audio/VO/TheImagined_DV, targets []}
   }}
 }]
properties {
 attenuationSettings=/BattlePassS20/Audio/VO/S20_Quest_Att
 maxDistance=2097152
 soundClassObject=/Game/Sounds/FortSoundClassesAndModes/Fort_VO_Narrator
 volumeMultiplier=2
}
